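; Quest actor for the Season 33 Battle Pass BlondeJaw quests that spawns a beacon when a matching side effect is applied to it.
; The side-effect event compares the trigger tag to the BlondeJaw Beacon quest event tag, then spawns BP_BlondeJaw_Beacon at this actor's transform.
; It ignores the source, target and trigger type; BeginPlay, overlap, tick and the construction script add no logic.
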(event EventBeginPlay
  (|CallParentFunction) ; This node is disabled and will not be called.\nDrag off pins to build functionality.
)

(event Collision|EventActorBeginOverlap
  (params (OtherActor :type Actor%20Object%20Reference))
  (|CallParentFunction OtherActor) ; This node is disabled and will not be called.\nDrag off pins to build functionality.
)

(event EventTick
  (params (DeltaSeconds :type Float%20%28single-precision%29))
  (|CallParentFunction DeltaSeconds) ; This node is disabled and will not be called.\nDrag off pins to build functionality.
)

(event SideEffects|EventBP_ApplySideEffect
  (params (ReactionSource :type Object%20Reference) (TargetObject :type Object%20Reference) (SideEffectTriggerTag :type Gameplay%20Tag%20Structure%20%28by%20ref%29) (SideEffectTriggerType :type EBlueprintableSideEffectTriggerType%20Enum%20%28by%20ref%29 :default "TriggeredAsTarget"))
  (if (== SideEffectTriggerTag "(TagName=\"Athena.Quests.S33.Event.BlondeJaw.Beacon\")")
    (Game|SpawnActorfromClass "/BattlePassS33_Quests/Items/QuestActors/BlondeJaw/Blueprints/BP_BlondeJaw_Beacon.BP_BlondeJaw_Beacon_C" (Transformation|GetActorTransform) "Undefined" "MultiplyWithRoot" 0 0 0 0)))

(fn ConstructionScript
  (|CallParentFunction))
